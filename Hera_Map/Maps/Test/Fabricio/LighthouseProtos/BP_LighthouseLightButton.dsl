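; Prototype lighthouse puzzle interactable that turns a statue mesh in 45-degree steps through a timeline on each "Rotate Light" interaction.
; When a turn finishes, the authority matches the yaw to one of eight secret data layers, activating it and dropping a neighbor to Loaded.
; RotatingLight's remote-only branch counts turns against BP_LighthouseLight_Proto's layer check, hardcoded to 6, for solved effects.
; Debug prints remain, and SetCurrentPuzzleDataIndex has empty rotation branches.
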(event Custom|EnableInteract
  (Collision|SetCollisionResponsetoChannel (Variables|Default|GetInteractionCapsule) "ECC_GameTraceChannel1" "ECR_Overlap")
  (Rendering|SetVisibility (Variables|Default|GetGlassEeffect) true false)
  (Rendering|SetVisibility (Variables|Default|GetSecretSpotlight) true false)
  (Rendering|SetVisibility (Variables|Default|GetGlassSpotlight) true false))

(event Interaction|EventBlueprintOnInteract
  (params (InteractingPawn :type Fort%20Pawn%20Object%20Reference) (InteractionBeingAttempted :type EInteractionBeingAttempted%20Enum :default "FirstInteraction"))
  ; Rotate On Interact
  (CallFunction|SetCurrentandNextRotation)
  (CallFunction|TurningClientEffects)
  (bind (_direction _turnstatue) (Timeline|TurnStatue PlayFromStart)
    (:Update
      (Transformation|SetRelativeRotation (Variables|Default|GetStatueMesh) (Math|Rotator|Lerp(Rotator) (Variables|Default|GetCurrentRelativeRotation) (Variables|Default|GetNextRelativeRotation) _turnstatue true) false false))
    (:Finished
      (Networking|SwitchHasAuthority
        (:Authority
          (Utilities|FlowControl|Delay 0.25)
          (Variables|Default|SetStatueIsTurning false)
          (Variables|Default|SetStatueCanTurn true)
          (CallFunction|SetCurrentandNextRotation)
          (CallFunction|SetCurrentPuzzleDataIndex)
          (CallFunction|RotatingLight)
          (Utilities|FlowControl|Delay 0.1)
          (if (== (Math|Float|Clamp(Float) (Variables|Default|GetPuzzleCurrentRotation) -1.0) 0)
            (DataLayers|SetDataLayerRuntimeState (WorldSubsystems|GetDataLayerSubsystem) (Variables|Default|GetSecretAssets_1) "Activated" false :!class /Script/Engine.DataLayerSubsystem)
            (DataLayers|SetDataLayerRuntimeState (WorldSubsystems|GetDataLayerSubsystem) (Variables|Default|GetSecretAssets_7) "Loaded" false :!class /Script/Engine.DataLayerSubsystem)
            (elif (== (Math|Float|Clamp(Float) (Variables|Default|GetPuzzleCurrentRotation) 44.0 46.0) 45.0)
              (DataLayers|SetDataLayerRuntimeState (WorldSubsystems|GetDataLayerSubsystem) (Variables|Default|GetSecretAssets_2) "Activated" false :!class /Script/Engine.DataLayerSubsystem)
              (DataLayers|SetDataLayerRuntimeState (WorldSubsystems|GetDataLayerSubsystem) (Variables|Default|GetSecretAssets_1) "Loaded" false :!class /Script/Engine.DataLayerSubsystem)
              (elif (== (Math|Float|Clamp(Float) (Variables|Default|GetPuzzleCurrentRotation) 89.0 91.0) 9e+01)
                (DataLayers|SetDataLayerRuntimeState (WorldSubsystems|GetDataLayerSubsystem) (Variables|Default|GetSecretAssets_3) "Activated" false :!class /Script/Engine.DataLayerSubsystem)
                (DataLayers|SetDataLayerRuntimeState (WorldSubsystems|GetDataLayerSubsystem) (Variables|Default|GetSecretAssets_2) "Loaded" false :!class /Script/Engine.DataLayerSubsystem)
                (elif (== (Math|Float|Clamp(Float) (Variables|Default|GetPuzzleCurrentRotation) 134.0 136.0) 135.0)
                  (DataLayers|SetDataLayerRuntimeState (WorldSubsystems|GetDataLayerSubsystem) (Variables|Default|GetSecretAssets_4) "Activated" false :!class /Script/Engine.DataLayerSubsystem)
                  (DataLayers|SetDataLayerRuntimeState (WorldSubsystems|GetDataLayerSubsystem) (Variables|Default|GetSecretAssets_3) "Loaded" false :!class /Script/Engine.DataLayerSubsystem)
                  (elif (== (Math|Float|Clamp(Float) (Variables|Default|GetPuzzleCurrentRotation) -181.0 -179.0) -1.8e+02)
                    (DataLayers|SetDataLayerRuntimeState (WorldSubsystems|GetDataLayerSubsystem) (Variables|Default|GetSecretAssets_5) "Activated" false :!class /Script/Engine.DataLayerSubsystem)
                    (DataLayers|SetDataLayerRuntimeState (WorldSubsystems|GetDataLayerSubsystem) (Variables|Default|GetSecretAssets_4) "Loaded" false :!class /Script/Engine.DataLayerSubsystem)
                    (elif (== (Math|Float|Clamp(Float) (Variables|Default|GetPuzzleCurrentRotation) -136.0 -134.0) -135.0)
                      (DataLayers|SetDataLayerRuntimeState (WorldSubsystems|GetDataLayerSubsystem) (Variables|Default|GetSecretAssets_6) "Activated" false :!class /Script/Engine.DataLayerSubsystem)
                      (DataLayers|SetDataLayerRuntimeState (WorldSubsystems|GetDataLayerSubsystem) (Variables|Default|GetSecretAssets_5) "Loaded" false :!class /Script/Engine.DataLayerSubsystem)
                      (elif (== (Math|Float|Clamp(Float) (Variables|Default|GetPuzzleCurrentRotation) -91.0 -89.0) -9e+01)
                        (DataLayers|SetDataLayerRuntimeState (WorldSubsystems|GetDataLayerSubsystem) (Variables|Default|GetSecretAssets_7) "Activated" false :!class /Script/Engine.DataLayerSubsystem)
                        (DataLayers|SetDataLayerRuntimeState (WorldSubsystems|GetDataLayerSubsystem) (Variables|Default|GetSecretAssets_6) "Loaded" false :!class /Script/Engine.DataLayerSubsystem)
                        (elif (== (Math|Float|Clamp(Float) (Variables|Default|GetPuzzleCurrentRotation) -46.0 -44.0) -45.0)
                          (DataLayers|SetDataLayerRuntimeState (WorldSubsystems|GetDataLayerSubsystem) (Variables|Default|GetSecretAssets_8) "Activated" false :!class /Script/Engine.DataLayerSubsystem)
                          (DataLayers|SetDataLayerRuntimeState (WorldSubsystems|GetDataLayerSubsystem) (Variables|Default|GetSecretAssets_7) "Loaded" false :!class /Script/Engine.DataLayerSubsystem)
                          (else
                            (Development|PrintString "ALL FALSE")))))))))))))))

(event Custom|TurningClientEffects
  ; Rotate On Interact
  (Networking|SwitchHasAuthority
    (:Remote
      (Audio|Components|Audio|Play (Variables|Default|GetAudioTurnCue) 0.0))))

(event Custom|TurningMagicLightOn
  (Networking|SwitchHasAuthority
    (:Authority
      (Rendering|SetVisibility (Variables|Default|GetGlassEeffect) true false)
      (Rendering|SetVisibility (Variables|Default|GetSecretSpotlight) true false))))

(event Custom|RotatingLight
  ; Puzzle Solved
  (Networking|SwitchHasAuthority
    (:Remote
      (bind _result (Math|Integer|IncrementInt (Variables|Default|GetPuzzleCurrentRotationForSecretSpawn))
        (:%20%20
          (Variables|Default|SetPuzzleCurrentRotationForSecretSpawn _result)
          (Variables|/Hera_Map/Maps/Test/Fabricio/LighthouseProtos/BP_LighthouseLight_Proto.BP_LighthouseLight_Proto_C|SetChosenDataLayerCheck 6 (Variables|Default|GetLighthouseLightReference))
          (Development|PrintString (Utilities|String|ToString(Integer) (Variables|/Hera_Map/Maps/Test/Fabricio/LighthouseProtos/BP_LighthouseLight_Proto.BP_LighthouseLight_Proto_C|GetChosenDataLayerCheck (Variables|Default|GetLighthouseLightReference))))
          (Utilities|FlowControl|Delay 0.5)
          (if (== (Variables|Default|GetPuzzleCurrentRotationForSecretSpawn) 1)
            (if (== (Variables|/Hera_Map/Maps/Test/Fabricio/LighthouseProtos/BP_LighthouseLight_Proto.BP_LighthouseLight_Proto_C|GetChosenDataLayerCheck (Variables|Default|GetLighthouseLightReference)) 0)
              (Niagara|SpawnSystematLocation "/Hermes/Blueprint/Prototypes/ChrisK/FX/NS_PuzzleTurner_Solved.NS_PuzzleTurner_Solved" (+ (Transformation|GetWorldLocation (Variables|Default|GetDefaultSceneRoot)) "0.000000,-20.000000,140.000000"))
              (Audio|SpawnSoundatLocation "/Hermes/Audio/Wolf_Puzzle/Interactables_S34_Wolf_Puzzle_Stinger_Cue.Interactables_S34_Wolf_Puzzle_Stinger_Cue" (Transformation|GetWorldLocation (Variables|Default|GetStatueMesh)))
              (Niagara|SpawnSystematLocation :SystemTemplate "/Game/Athena/Items/Consumables/KnockGrenade/FX/NS_Athena_KnockGrenade_Explosion.NS_Athena_KnockGrenade_Explosion" :Location (Transformation|GetWorldLocation (Variables|Default|GetFXSecretSpawn_1)) :Scale (Variables|Default|GetSpawnEffectSize) :bAutoDestroy true :bAutoActivate true :PoolingMethod "None" :bPreCullCheck true)
              (else
                (Development|PrintString "NOT 1")))
            (elif (== (Variables|Default|GetPuzzleCurrentRotationForSecretSpawn) 2)
              (if (== (Variables|/Hera_Map/Maps/Test/Fabricio/LighthouseProtos/BP_LighthouseLight_Proto.BP_LighthouseLight_Proto_C|GetChosenDataLayerCheck (Variables|Default|GetLighthouseLightReference)) 1)
                (Niagara|SpawnSystematLocation "/Hermes/Blueprint/Prototypes/ChrisK/FX/NS_PuzzleTurner_Solved.NS_PuzzleTurner_Solved" (+ (Transformation|GetWorldLocation (Variables|Default|GetDefaultSceneRoot)) "0.000000,-20.000000,140.000000"))
                (Audio|SpawnSoundatLocation "/Hermes/Audio/Wolf_Puzzle/Interactables_S34_Wolf_Puzzle_Stinger_Cue.Interactables_S34_Wolf_Puzzle_Stinger_Cue" (Transformation|GetWorldLocation (Variables|Default|GetStatueMesh)))
                (Niagara|SpawnSystematLocation :SystemTemplate "/Game/Athena/Items/Consumables/KnockGrenade/FX/NS_Athena_KnockGrenade_Explosion.NS_Athena_KnockGrenade_Explosion" :Location (Transformation|GetWorldLocation (Variables|Default|GetFXSecretSpawn_2)) :Scale (Variables|Default|GetSpawnEffectSize) :bAutoDestroy true :bAutoActivate true :PoolingMethod "None" :bPreCullCheck true)
                (else
                  (Development|PrintString "NOT 2")))
              (elif (== (Variables|Default|GetPuzzleCurrentRotationForSecretSpawn) 3)
                (if (== (Variables|/Hera_Map/Maps/Test/Fabricio/LighthouseProtos/BP_LighthouseLight_Proto.BP_LighthouseLight_Proto_C|GetChosenDataLayerCheck (Variables|Default|GetLighthouseLightReference)) 2)
                  (Niagara|SpawnSystematLocation "/Hermes/Blueprint/Prototypes/ChrisK/FX/NS_PuzzleTurner_Solved.NS_PuzzleTurner_Solved" (+ (Transformation|GetWorldLocation (Variables|Default|GetDefaultSceneRoot)) "0.000000,-20.000000,140.000000"))
                  (Audio|SpawnSoundatLocation "/Hermes/Audio/Wolf_Puzzle/Interactables_S34_Wolf_Puzzle_Stinger_Cue.Interactables_S34_Wolf_Puzzle_Stinger_Cue" (Transformation|GetWorldLocation (Variables|Default|GetStatueMesh)))
                  (Niagara|SpawnSystematLocation :SystemTemplate "/Game/Athena/Items/Consumables/KnockGrenade/FX/NS_Athena_KnockGrenade_Explosion.NS_Athena_KnockGrenade_Explosion" :Location (Transformation|GetWorldLocation (Variables|Default|GetFXSecretSpawn_3)) :Scale (Variables|Default|GetSpawnEffectSize) :bAutoDestroy true :bAutoActivate true :PoolingMethod "None" :bPreCullCheck true)
                  (else
                    (Development|PrintString "NOT 3")))
                (elif (== (Variables|Default|GetPuzzleCurrentRotationForSecretSpawn) 4)
                  (if (== (Variables|/Hera_Map/Maps/Test/Fabricio/LighthouseProtos/BP_LighthouseLight_Proto.BP_LighthouseLight_Proto_C|GetChosenDataLayerCheck (Variables|Default|GetLighthouseLightReference)) 3)
                    (Niagara|SpawnSystematLocation "/Hermes/Blueprint/Prototypes/ChrisK/FX/NS_PuzzleTurner_Solved.NS_PuzzleTurner_Solved" (+ (Transformation|GetWorldLocation (Variables|Default|GetDefaultSceneRoot)) "0.000000,-20.000000,140.000000"))
                    (Audio|SpawnSoundatLocation "/Hermes/Audio/Wolf_Puzzle/Interactables_S34_Wolf_Puzzle_Stinger_Cue.Interactables_S34_Wolf_Puzzle_Stinger_Cue" (Transformation|GetWorldLocation (Variables|Default|GetStatueMesh)))
                    (Niagara|SpawnSystematLocation :SystemTemplate "/Game/Athena/Items/Consumables/KnockGrenade/FX/NS_Athena_KnockGrenade_Explosion.NS_Athena_KnockGrenade_Explosion" :Location (Transformation|GetWorldLocation (Variables|Default|GetFXSecretSpawn_4)) :Scale (Variables|Default|GetSpawnEffectSize) :bAutoDestroy true :bAutoActivate true :PoolingMethod "None" :bPreCullCheck true)
                    (else
                      (Development|PrintString "NOT 4")))
                  (elif (== (Variables|Default|GetPuzzleCurrentRotationForSecretSpawn) 5)
                    (if (== (Variables|/Hera_Map/Maps/Test/Fabricio/LighthouseProtos/BP_LighthouseLight_Proto.BP_LighthouseLight_Proto_C|GetChosenDataLayerCheck (Variables|Default|GetLighthouseLightReference)) 4)
                      (Niagara|SpawnSystematLocation "/Hermes/Blueprint/Prototypes/ChrisK/FX/NS_PuzzleTurner_Solved.NS_PuzzleTurner_Solved" (+ (Transformation|GetWorldLocation (Variables|Default|GetDefaultSceneRoot)) "0.000000,-20.000000,140.000000"))
                      (Audio|SpawnSoundatLocation "/Hermes/Audio/Wolf_Puzzle/Interactables_S34_Wolf_Puzzle_Stinger_Cue.Interactables_S34_Wolf_Puzzle_Stinger_Cue" (Transformation|GetWorldLocation (Variables|Default|GetStatueMesh)))
                      (Niagara|SpawnSystematLocation :SystemTemplate "/Game/Athena/Items/Consumables/KnockGrenade/FX/NS_Athena_KnockGrenade_Explosion.NS_Athena_KnockGrenade_Explosion" :Location (Transformation|GetWorldLocation (Variables|Default|GetFXSecretSpawn_5)) :Scale (Variables|Default|GetSpawnEffectSize) :bAutoDestroy true :bAutoActivate true :PoolingMethod "None" :bPreCullCheck true)
                      (else
                        (Development|PrintString "NOT 5")))
                    (elif (== (Variables|Default|GetPuzzleCurrentRotationForSecretSpawn) 6)
                      (if (== (Variables|/Hera_Map/Maps/Test/Fabricio/LighthouseProtos/BP_LighthouseLight_Proto.BP_LighthouseLight_Proto_C|GetChosenDataLayerCheck (Variables|Default|GetLighthouseLightReference)) 6)
                        (Niagara|SpawnSystematLocation "/Hermes/Blueprint/Prototypes/ChrisK/FX/NS_PuzzleTurner_Solved.NS_PuzzleTurner_Solved" (+ (Transformation|GetWorldLocation (Variables|Default|GetDefaultSceneRoot)) "0.000000,-20.000000,140.000000"))
                        (Audio|SpawnSoundatLocation "/Hermes/Audio/Wolf_Puzzle/Interactables_S34_Wolf_Puzzle_Stinger_Cue.Interactables_S34_Wolf_Puzzle_Stinger_Cue" (Transformation|GetWorldLocation (Variables|Default|GetStatueMesh)))
                        (Niagara|SpawnSystematLocation :SystemTemplate "/Game/Athena/Items/Consumables/KnockGrenade/FX/NS_Athena_KnockGrenade_Explosion.NS_Athena_KnockGrenade_Explosion" :Location (Transformation|GetWorldLocation (Variables|Default|GetFXSecretSpawn_6)) :Scale (Variables|Default|GetSpawnEffectSize) :bAutoDestroy true :bAutoActivate true :PoolingMethod "None" :bPreCullCheck true)
                        (else
                          (Development|PrintString "NOT 6")))
                      (elif (== (Variables|Default|GetPuzzleCurrentRotationForSecretSpawn) 7)
                        (if (== (Variables|/Hera_Map/Maps/Test/Fabricio/LighthouseProtos/BP_LighthouseLight_Proto.BP_LighthouseLight_Proto_C|GetChosenDataLayerCheck (Variables|Default|GetLighthouseLightReference)) 6)
                          (Niagara|SpawnSystematLocation "/Hermes/Blueprint/Prototypes/ChrisK/FX/NS_PuzzleTurner_Solved.NS_PuzzleTurner_Solved" (+ (Transformation|GetWorldLocation (Variables|Default|GetDefaultSceneRoot)) "0.000000,-20.000000,140.000000"))
                          (Audio|SpawnSoundatLocation "/Hermes/Audio/Wolf_Puzzle/Interactables_S34_Wolf_Puzzle_Stinger_Cue.Interactables_S34_Wolf_Puzzle_Stinger_Cue" (Transformation|GetWorldLocation (Variables|Default|GetStatueMesh)))
                          (Niagara|SpawnSystematLocation :SystemTemplate "/Game/Athena/Items/Consumables/KnockGrenade/FX/NS_Athena_KnockGrenade_Explosion.NS_Athena_KnockGrenade_Explosion" :Location (Transformation|GetWorldLocation (Variables|Default|GetFXSecretSpawn_7)) :Scale (Variables|Default|GetSpawnEffectSize) :bAutoDestroy true :bAutoActivate true :PoolingMethod "None" :bPreCullCheck true)
                          (else
                            (Development|PrintString "NOT 7")))
                        (elif (== (Variables|Default|GetPuzzleCurrentRotationForSecretSpawn) 8)
                          (if (== (Variables|/Hera_Map/Maps/Test/Fabricio/LighthouseProtos/BP_LighthouseLight_Proto.BP_LighthouseLight_Proto_C|GetChosenDataLayerCheck (Variables|Default|GetLighthouseLightReference)) 7)
                            (Niagara|SpawnSystematLocation "/Hermes/Blueprint/Prototypes/ChrisK/FX/NS_PuzzleTurner_Solved.NS_PuzzleTurner_Solved" (+ (Transformation|GetWorldLocation (Variables|Default|GetDefaultSceneRoot)) "0.000000,-20.000000,140.000000"))
                            (Audio|SpawnSoundatLocation "/Hermes/Audio/Wolf_Puzzle/Interactables_S34_Wolf_Puzzle_Stinger_Cue.Interactables_S34_Wolf_Puzzle_Stinger_Cue" (Transformation|GetWorldLocation (Variables|Default|GetStatueMesh)))
                            (Niagara|SpawnSystematLocation :SystemTemplate "/Game/Athena/Items/Consumables/KnockGrenade/FX/NS_Athena_KnockGrenade_Explosion.NS_Athena_KnockGrenade_Explosion" :Location (Transformation|GetWorldLocation (Variables|Default|GetFXSecretSpawn_8)) :Scale (Variables|Default|GetSpawnEffectSize) :bAutoDestroy true :bAutoActivate true :PoolingMethod "None" :bPreCullCheck true)
                            (else
                              (Development|PrintString "NOT 8")))
                          (else
                            (Development|PrintString "ALL FAILED TOTAL")))))))))))))))

(event Custom|PuzzleIsSolved
  (bind _defaultsceneroot (Variables|Default|GetDefaultSceneRoot))
  (bind _returnvalue (Transformation|GetWorldLocation _defaultsceneroot))
  (bind _returnvalue_2 (+ _returnvalue "0.000000,-20.000000,140.000000"))
  (Networking|SwitchHasAuthority
    (:Remote
      (Niagara|SpawnSystematLocation "/Hermes/Blueprint/Prototypes/ChrisK/FX/NS_PuzzleTurner_Solved.NS_PuzzleTurner_Solved" _returnvalue_2)
      (Audio|SpawnSoundatLocation "/Hermes/Audio/Wolf_Puzzle/Interactables_S34_Wolf_Puzzle_Stinger_Cue.Interactables_S34_Wolf_Puzzle_Stinger_Cue" _returnvalue_2))))

(fn BlueprintCanInteract
  (params (InteractingPawn :type Fort%20Pawn%20Object%20Reference) (InteractionBeingAttempted :type EInteractionBeingAttempted%20Enum :default "FirstInteraction") (InteractionType :type TInteractionType%20Enum :default "IT_NoInteraction"))
  (return (!= false (Variables|Default|GetStatueCanTurn))))

(fn BlueprintGetInteractionString
  (params (InteractingPawn :type Fort%20Pawn%20Object%20Reference) (InteractionBeingAttempted :type EInteractionBeingAttempted%20Enum :default "FirstInteraction"))
  (return "NSLOCTEXT(\"[B5E9BDE246C24E903ACB941ED6AB1525]\", \"909617184962222E1B1CF4A4A1C295ED\", \"Rotate Light\")"))

(fn SetCurrentAndNextRotation
  (bind _output_get (Variables|Default|SetCurrentRelativeRotation (Variables|/Script/Engine.SceneComponent|GetRelativeRotation (Variables|Default|GetStatueMesh))))
  (if (== (.yaw _output_get) 315.0)
    (Variables|Default|SetNextRelativeRotation "0, 0, 0")
    (else
      (Variables|Default|SetNextRelativeRotation (Math|Rotator|CombineRotators _output_get (Math|Rotator|MakeRotator 0.0 0.0 45.0)))
      (bind _output_get_2 (Variables|Default|SetPuzzleCurrentRotation (.yaw _output_get)))
      (Development|PrintString (Utilities|String|ToString(Float) _output_get_2)))))

(fn SetCurrentPuzzleDataIndex
  (if (Utilities|Operators|Equal(==) :A (Variables|Default|GetCurrentRelativeRotation) :ErrorTolerance 1.0)
    _
    (elif (Utilities|Operators|Equal(==) (Variables|Default|GetCurrentRelativeRotation) "0.000000,45.000000,0.000000" 1.0)
      _
      (elif (Utilities|Operators|Equal(==) (Variables|Default|GetCurrentRelativeRotation) "0.000000,90.000000,0.000000" 1.0)
        _
        (elif (Utilities|Operators|Equal(==) (Variables|Default|GetCurrentRelativeRotation) "0.000000,135.000000,0.000000" 1.0)
          _
          (elif (Utilities|Operators|Equal(==) (Variables|Default|GetCurrentRelativeRotation) "0.000000,180.000000,0.000000" 1.0)
            _
            (elif (Utilities|Operators|Equal(==) (Variables|Default|GetCurrentRelativeRotation) "0.000000,225.000000,0.000000" 1.0)
              _
              (elif (Utilities|Operators|Equal(==) (Variables|Default|GetCurrentRelativeRotation) "0.000000,270.000000,0.000000" 1.0)
                _
                (elif (Utilities|Operators|Equal(==) (Variables|Default|GetCurrentRelativeRotation) "0.000000,315.000000,0.000000" 1.0)
                  _
                  (else
                    (Development|PrintString "Hello")))))))))))

(fn ConstructionScript)
