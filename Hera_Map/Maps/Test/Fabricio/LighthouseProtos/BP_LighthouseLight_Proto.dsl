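; Prototype lighthouse light Blueprint that switches on a data layer when play begins.
; BeginPlay safely fetches the hard-coded seventh entry (index 6) of its data layer list and sets it to Activated through the DataLayerSubsystem.
; It then records that index in ChosenDataLayerCheck and prints it with a "Hello" debug message.
; Overlap, Tick and construction script are empty stubs, and no light behavior appears in the graph.
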(event EventBeginPlay
  (bind _result (ArrayMacros|SafeArrayGet (Variables|Default|GetDataLayerList) 6)
    (:%20%20
      (DataLayers|SetDataLayerRuntimeState (WorldSubsystems|GetDataLayerSubsystem) _result "Activated" false :!class /Script/Engine.DataLayerSubsystem)
      (bind _output_get (Variables|Default|SetChosenDataLayerCheck 6))
      (Development|PrintString "Hello" true true "(R=1.000000,G=0.000000,B=0.000938,A=1.000000)")
      (Development|PrintString (Utilities|String|ToString(Integer) _output_get) true true "(R=0.884252,G=0.000000,B=1.000000,A=1.000000)"))))

(event Collision|EventActorBeginOverlap
  (params (OtherActor :type Actor%20Object%20Reference)))

(event EventTick
  (params (DeltaSeconds :type Float%20%28single-precision%29)))

(fn ConstructionScript)
