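; Lighthouse prototype Blueprint that rotates the whole actor when its TurnLighthouseOn custom event is called.
; The event first sets the BP_FX_LightHouse component to hidden in game, then plays a timeline.
; Each timeline update adds a world yaw rotation scaled from the light rotation loop track, with no delta-time term, so spin speed follows update rate.
; The construction script is empty.
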(event Custom|TurnLighthouseOn
  (Development|SetHiddeninGame (Variables|Default|GetBP_FX_LightHouse) true false)
  (bind (_direction _lightrotationloop) (Timeline|Timeline Play)
    (:Update
      (Transformation|AddActorWorldRotation (Math|Rotator|MakeRotator 0.0 0.0 (* _lightrotationloop 1.5))))))

(fn ConstructionScript)
