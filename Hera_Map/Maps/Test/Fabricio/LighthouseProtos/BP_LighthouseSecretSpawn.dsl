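; Lighthouse prototype Blueprint that plays a reveal effect when its PuzzleSolved custom event is called.
; After a short delay, it spawns a borrowed Knock Grenade explosion Niagara effect and a Wolf puzzle stinger sound at its Arrow component's world location.
; It then prints the root component's display name for debugging; despite the SecretSpawn name, no actor is spawned here.
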(event Custom|PuzzleSolved
  (bind _arrow (Variables|Default|GetArrow))
  (bind _returnvalue (Transformation|GetWorldLocation _arrow))
  (Utilities|FlowControl|Delay 2.0)
  (Niagara|SpawnSystematLocation "/Game/Athena/Items/Consumables/KnockGrenade/FX/NS_Athena_KnockGrenade_Explosion.NS_Athena_KnockGrenade_Explosion" _returnvalue)
  (Audio|SpawnSoundatLocation "/Hermes/Audio/Wolf_Puzzle/Interactables_S34_Wolf_Puzzle_Stinger_Cue.Interactables_S34_Wolf_Puzzle_Stinger_Cue" _returnvalue)
  (Development|PrintString (Utilities|GetDisplayName (Variables|Default|GetDefaultSceneRoot))))

(fn ConstructionScript)
